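# Translucent unlit material for the engine flame on the Inferno glider, usable on Niagara sprites, mesh particles and skeletal meshes.
# Panning noise layers drive emissive color over Fresnel-blended HDR tints, with an optional scrolling nebula-and-star overlay.
# Opacity combines that noise with V gradients, invertible Fresnel, particle alpha and depth fade.
# World position offset adds vertex-color-varied sine wobble, noise displacement and a normal bulge, scalable by dynamic parameter.
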Material /Game/Gadgets/Assets/VinderTech_GliderChute/Glider_Inferno/FX/M_Glider_Inferno_EngineFlame
graph {
 properties {
  bUsedWithMeshDeformer=true
  bUsedWithNiagaraMeshParticles=true
  bUsedWithNiagaraSprites=true
  bUsedWithSkeletalMesh=true
  blendMode=BLEND_Translucent
  pixelDepthOffsetMode=PDOM_Legacy
  shadingModel=MSM_Unlit
  usedShadingModels=Unlit
 }
 nodes [
  n1 = Constant()
  n2 = Time()
  n3 = VertexColor()
  n4 = LinearInterpolate(Alpha=@n3.R) {constB=0.5}
  n5 = Sine(None=Add(A=Multiply(A=@n2, B=@n4), B=LinearInterpolate(Alpha=@n3.R) {constB=2}))
  n6 = AppendVector(A=@n5, B=@n5)
  n7 = Transform(None=Multiply(A=AppendVector(A=@n1, B=@n6), B=VectorParameter() {group=WPO, parameterName=FlameDirection, defaultValue=<rgba 0 -1 1 1>}.RGB) {constA=1}) {transformSourceType=TRANSFORMSOURCE_Local}
  n8 = ComponentMask(None=TextureCoordinate()) {g=true}
  n9 = LinearInterpolate(
   B=ScalarParameter() {defaultValue=9, group=WPO, parameterName=Distort}
   Alpha=Power(Base=OneMinus(None=@n8), Exp=ScalarParameter() {defaultValue=3, group=WPO, parameterName=WPOContrast})
  )
  n10 = OneMinus(None=ComponentMask(None=TextureCoordinate()) {g=true})
  n11 = LinearInterpolate(B=ScalarParameter() {defaultValue=8, group=WPO, parameterName=Distort Edge}, Alpha=Multiply(A=@n10, B=@n10))
  n12 = ComponentMask(None=VectorParameter() {group=FlameShape, parameterName=Flames_UVScale, defaultValue=<rgba 0.6 0.4 0 0>}.RGB) {g=true, r=true}
  n13 = TextureCoordinate() {uTiling=4}
  n14 = ScalarParameter() {group=FlameShape, parameterName=Flame_UVOffset}
  n15 = Multiply(A=Time(), B=ScalarParameter() {defaultValue=3, group=FlameShape, parameterName=Flame_Speed})
  n16 = Panner(Coordinate=Add(A=Multiply(A=@n12, B=@n13), B=@n14), Time=@n15) {speedX=0.2, speedY=1}
  n17 = Multiply(
   A=LinearInterpolate(Alpha=TextureSample!(UVs=@n16, Tex, Level, Apply View MipBias) {mipValueMode=TMVM_MipLevel, texture=&T_NoisePack_01}.G) {constA=-1}
   B=VertexNormalWS()
  )
  n18 = ComponentMask(None=TextureCoordinate()) {g=true}
  n19 = ScalarParameter() {defaultValue=-0.25, group=WPO, parameterName=Distort_offset}
  n20 = Power(Base=Sine(None=Add(A=@n18, B=@n19)) {period=2}, Exp=ScalarParameter() {defaultValue=6, group=WPO, parameterName=Distort_power})
  n21 = ScalarParameter() {defaultValue=2, group=WPO, parameterName=Distort_Multi}
  n22 = Add(A=Add(A=Multiply(A=@n7, B=@n9), B=Multiply(A=@n11, B=@n17)), B=Multiply(A=VertexNormalWS(), B=Multiply(A=@n20, B=@n21))) {constB=0}
  LinearInterpolate(
   A=Multiply(A=@n22, B=ScalarParameter() {defaultValue=1, group=WPO, parameterName=WPO_OverallIntensity})
   B=Multiply(A=@n22, B=DynamicParameter() {defaultValue=<rgba 1 0 0 0>, paramNames [WPO Intensity, Param2, Param3, Param4]}.WPO Intensity)
   Alpha=ScalarParameter() {defaultValue=1, group=WPO, parameterName=Use Dynamic WPO}
  ) {constA=1}
  n23 = Power(
   Base=ComponentMask(None=TextureCoordinate()) {g=true}
   Exp=ScalarParameter() {defaultValue=3.5, group=Gradient, parameterName=Gradient Power}
  )
  n24 = ScalarParameter() {defaultValue=100, group=Gradient, parameterName=Gradient Mult}
  n25 = LinearInterpolate(
   A=VectorParameter() {group=Emissive, parameterName=Emissive Color A, defaultValue=<rgba 0.0554 0.015 0.5 1>}.RGB
   B=VectorParameter() {group=Emissive, parameterName=Emissive Color B, defaultValue=<rgba 0.035 0.005 0.05 1>}.RGB
   Alpha=Fresnel(
    ExponentIn=ScalarParameter() {defaultValue=0.2, group=Fresnel, parameterName=Emissive Fresnel Exp}
    BaseReflectFractionIn=ScalarParameter() {group=Fresnel, parameterName=Emissive Fresnel Base}
   ) {baseReflectFraction=0, exponent=1}
  )
  n26 = VectorParameter() {group=Emissive, parameterName=Color_HDR, defaultValue=<rgba 0 0.4167 5 1>}
  n27 = Multiply(A=TextureCoordinate() {vTiling=0.5}, B=@n12)
  n28 = TextureSample!(UVs=Panner(Coordinate=Add(A=@n27, B=@n14)) {speedX=-0.5, speedY=1}, Tex, Apply View MipBias) {texture=&T_NoisePack_01}
  n29 = Constant() {r=0.2}
  n30 = Add(A=@n14, B=Multiply(A=@n12, B=TextureCoordinate() {uTiling=6, vTiling=1.5}))
  n31 = Multiply(
   A=Add(
    A=MaterialFunctionCall(In=@n28.G, Contrast=@n29) {materialFunction=&CheapContrast}.Result
    B=TextureSample(UVs=@n16) {texture=&T_NoisePack_01}.G
   )
   B=TextureSample(UVs=Panner(Coordinate=@n30, Time=@n15) {speedX=-0.2, speedY=1}) {texture=&T_NoisePack_01}.G
  )
  n32 = Multiply(
   A=MaterialFunctionCall(
    LDR Input=Clamp(None=@n25)
    HDR Tint=Multiply(A=@n25, B=@n26.RGB)
    HDR Intensity Multiplier=ScalarParameter() {defaultValue=3, group=Emissive, parameterName=HDR_Intensity}
   ) {materialFunction=&DeriveHDRfromLDR}.RemappedRange
   B=LinearInterpolate(
    A=VectorParameter() {group=Emissive, parameterName=Emissive Dark Color, defaultValue=<rgba 1 1 1 1>}.RGB
    B=VectorParameter() {group=Emissive, parameterName=Emissive Light Color, defaultValue=<rgba 0 0 4 1>}.RGB
    Alpha=@n31
   )
  )
  n33 = LinearInterpolate(A=Multiply(A=Multiply(A=@n23, B=@n24), B=@n32), B=@n32, Alpha=ScalarParameter() {defaultValue=0.5, parameterName=Alpha})
  n34 = Panner(
   Coordinate=MaterialFunctionCall(UVs=TextureCoordinate() {uTiling=2, vTiling=2}, Texture Scale=ScalarParameter() {defaultValue=1, group=Galaxy, parameterName=Scale Galaxy}) {materialFunction=&ScaleUVsByCenter}.UVs
   Speed=AppendVector(A=Constant(), B=ScalarParameter() {defaultValue=1, group=Galaxy, parameterName=Speed Galaxy})
  ) {speedY=1}
  n35 = Multiply(
   A=VectorParameter() {group=Galaxy, parameterName=Galaxy Color, defaultValue=<rgba 1 1 1 1>}.RGB
   B=Desaturation(
    None=TextureSample(UVs=@n34) {texture=&/T_FN_Nebula_SmallSize}.RGB
    Fraction=ScalarParameter() {defaultValue=-20, group=Galaxy, parameterName=Galaxy Saturation}
   )
  )
  n36 = Desaturation(None=@n35, Fraction=ScalarParameter() {defaultValue=0.5, group=Galaxy, parameterName=Star Saturation})
  n37 = Multiply(A=@n36, B=VectorParameter() {group=Galaxy, parameterName=Star Brightness, defaultValue=<rgba 300 300 300 1>}.RGB)
  n38 = TextureCoordinate() {uTiling=2, vTiling=2}
  n39 = ScalarParameter() {defaultValue=1, group=Galaxy, parameterName=Scale Star 1}
  n40 = Constant()
  n41 = ScalarParameter() {defaultValue=1, group=Galaxy, parameterName=Speed Star 1}
  n42 = TextureSample(UVs=Panner(
    Coordinate=MaterialFunctionCall(UVs=@n38, Texture Scale=@n39) {materialFunction=&ScaleUVsByCenter}.UVs
    Speed=AppendVector(A=@n40, B=@n41)
   ) {speedY=0.5}) {samplerType=SAMPLERTYPE_Masks, texture=/Game/Characters/Player/Male/Medium/Bodies/M_Med_Soldier_04/Skins/Galaxy/Textures/Celestial_T_Stars}
  n43 = ComponentMask(None=@n42.RGB) {b=true}
  n44 = Multiply(A=@n43, B=@n43)
  n45 = VectorParameter() {group=Galaxy, parameterName=Star Brightness_1, defaultValue=<rgba 950 950 950 1>}
  n46 = TextureCoordinate() {uTiling=2, vTiling=2}
  n47 = ScalarParameter() {defaultValue=1, group=Galaxy, parameterName=Scale Star 2}
  n48 = Constant()
  n49 = ScalarParameter() {defaultValue=1, group=Galaxy, parameterName=Speed Star 2}
  n50 = TextureSample(UVs=Panner(
    Coordinate=MaterialFunctionCall(UVs=@n46, Texture Scale=@n47) {materialFunction=&ScaleUVsByCenter}.UVs
    Speed=AppendVector(A=@n48, B=@n49)
   ) {speedY=1}) {samplerType=SAMPLERTYPE_Masks, texture=&Celestial_T_Stars}
  n51 = ComponentMask(None=@n50.RGB) {r=true}
  n52 = Multiply(A=@n51, B=@n51)
  n53 = TextureSample(UVs=Panner(Coordinate=TextureCoordinate() {uTiling=2, vTiling=2}) {speedX=0.133, speedY=0.5}) {
   samplerSource=SSM_Wrap_WorldGroupSettings
   samplerType=SAMPLERTYPE_Masks
   texture=&Celestial_T_Stars
  }
  n54 = Multiply(
   A=LinearInterpolate(A=LinearInterpolate(A=@n35, B=@n37, Alpha=@n44), B=Multiply(A=@n36, B=@n45.RGB), Alpha=Multiply(A=@n52, B=@n53.G))
   B=@n33
  )
  n55 = ScalarParameter() {group=Galaxy, parameterName=Lerp Galaxy, sortPriority=0}
  n56 = ParticleColor()
  MaterialFunctionCall(Emissive=Multiply(A=LinearInterpolate(A=@n33, B=@n54, Alpha=@n55), B=@n56.RGB)) {materialFunction=&MF_FX_EmissiveAdjustments}
  n57 = ScalarParameter() {defaultValue=20, group=Alpha, parameterName=MasterAlpha}
  n58 = MaterialFunctionCall() {materialFunction=&LinearGradient}
  n59 = ScalarParameter() {defaultValue=0.75, group=FlameShape, parameterName=FireLine}
  n60 = Power(Base=Add(A=@n31, B=Power(Base=@n58.VGradient, Exp=@n59))) {constExponent=0.5}
  n61 = MaterialFunctionCall() {materialFunction=&LinearGradient}
  n62 = ScalarParameter() {defaultValue=2, group=Alpha, parameterName=VGradientContrast_Top}
  n63 = OneMinus(None=ComponentMask(None=TextureCoordinate()) {g=true})
  n64 = ScalarParameter() {defaultValue=3.5, group=Alpha, parameterName=UCoordAlphaFade}
  n65 = Multiply(
   A=Multiply(A=Power(Base=@n61.VGradient, Exp=@n62), B=Power(Base=@n63, Exp=@n64))
   B=ScalarParameter() {defaultValue=2, group=Alpha, parameterName=Enhance}
  )
  n66 = Power(
   Base=Fresnel(
    ExponentIn=ScalarParameter() {defaultValue=0.2, group=Fresnel, parameterName=Alpha Fresnel Exp}
    BaseReflectFractionIn=ScalarParameter() {group=Fresnel, parameterName=Alpha Fresnel Base}
   ) {baseReflectFraction=0, exponent=1}
   Exp=ScalarParameter() {defaultValue=1, group=Fresnel, parameterName=Alpha Fresnel Power}
  )
  n67 = Multiply(
   A=Multiply(A=Clamp(None=@n60), B=Saturate(None=@n65))
   B=StaticSwitchParameter(True=OneMinus(None=@n66), False=@n66) {group=Fresnel, parameterName=Invert Alpha Fresnel}
  )
  Multiply(
   A=@n56.A
   B=DepthFade(Opacity=Multiply(A=@n57, B=@n67), FadeDistance=ScalarParameter() {defaultValue=2, group=Alpha, parameterName=Fade Distance})
  )
  OneMinus(None=@n61.VGradient)
 ]
}
